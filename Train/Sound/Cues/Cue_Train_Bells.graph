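# Sound cue for the train's exterior bells, assigned to the exterior vehicles sound class.
# A looping wave player plays the MSS_Train_Exterior_Bells asset through a modulator with pitch and volume fixed at unity, into an attenuation node using the train bells settings.
# The cue routes through the train source effect chain.
# Its concurrency set references first-person weapon gunfire concurrency, and the debug flag is on.
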SoundCue /Train/Sound/Cues/Cue_Train_Bells
root=Attenuation_5
nodes [
 {id=Attenuation_5, children [Modulator_6], properties {attenuationSettings=/Train/Sound/Settings/Train_Bells_Att}}
 {id=Modulator_6, children [WavePlayer_0], properties {pitchMax=1, pitchMin=1, volumeMax=1, volumeMin=1}}
 {id=WavePlayer_0, properties {bLooping=true, soundWaveAssetPtr=/Train/Sound/MS/MSS_Train_Exterior_Bells}}
]
properties {
 bDebug=true
 maxDistance=6500
 soundClassObject=&Fort_Vehicles_Exterior
 sourceEffectChain=/Train/Sound/Settings/SEPC_Train
 volumeMultiplier=1.08
 attenuationOverrides {bSpatialize=false}
 concurrencyOverrides {bLimitToOwner=true, maxCount=3, resolutionRule=StopOldest}
 concurrencySet [&Weapon_Gunfire_FirstPerson_Conc]
}
