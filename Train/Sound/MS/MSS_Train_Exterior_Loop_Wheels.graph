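# Stereo MetaSound source for the train's exterior wheel audio, responding to movement speed, pitch and enclosure inputs.
# Two MSS_Train_Exterior_Template instances play the wheel variations and intermittent track squeaks, mixed with squeaks quieter.
# A trigger repeat fires squeaks at random intervals, feeding back through a deferred variable to reset the random time node.
# The CB_Train_Stopped modulator scales final volume; the Mobile page drops the squeak layer and mixer.
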MetaSoundSource /Train/Sound/MS/MSS_Train_Exterior_Loop_Wheels
inputs [
 Enclosed.LPF {type=Float}
 Enclosed.Volume {type=Float}
 Input {type=Float}
 Loop {type=Bool}
 Movement.Speed {type=Float}
 Movement.SpeedAzimuth {type=Float}
 Movement.SpeedElevation {type=Float}
 Pitch {type=Float}
 "Track Stress " {type=WaveAsset:Array}
 UE.Source.OnPlay {type=Trigger}
 "Wheels " {type=WaveAsset:Array}
]
interfaces [Fort.Movement:1.0, UE.OutputFormat.Stereo:1.0, UE.Source:1.1]
outputs [
 "UE.OutputFormat.Stereo.Audio:0" {type=Audio}
 "UE.OutputFormat.Stereo.Audio:1" {type=Audio}
]
pages [
 Default {
  is_build_graph=true
  input_defaults {
   Enclosed.LPF=6000
   Enclosed.Volume=0.5
   Input=0
   Loop="<MetaSoundBuilderResult.SUCCEEDED: 0>"
   Movement.Speed=1000
   Movement.SpeedAzimuth=0
   Movement.SpeedElevation=0
   Pitch=0.8
   UE.Source.OnPlay=null
   "Track Stress " [
    /Train/Sound/Waves/Locomotion/S28/Train_TrackSqueak_01
    /Train/Sound/Waves/Locomotion/S28/Train_TrackSqueak_02
    /Train/Sound/Waves/Locomotion/S28/Train_TrackSqueak_03
    /Train/Sound/Waves/Locomotion/S28/Train_TrackSqueak_04
    /Train/Sound/Waves/Locomotion/S28/Train_TrackSqueak_05
    /Train/Sound/Waves/Locomotion/S28/Train_TrackSqueak_06
    /Train/Sound/Waves/Locomotion/S28/Train_TrackSqueak_07
    /Train/Sound/Waves/Locomotion/S28/Train_TrackSqueak_08
    /Train/Sound/Waves/Locomotion/S28/Train_TrackSqueak_09
    /Train/Sound/Waves/Locomotion/S28/Train_TrackSqueak_10
    /Train/Sound/Waves/Locomotion/S28/Train_TrackSqueak_11
   ]
   "Wheels " [
    /Train/Sound/Waves/Locomotion/S28/Train_Proto_Movement_Wheels_02
    /Train/Sound/Waves/Locomotion/S28/Train_Proto_Movement_Wheels_03
    /Train/Sound/Waves/Locomotion/S28/Train_Proto_Movement_Wheels_04
   ]
  }
  nodes [
   n1 = VariableDeferredAccessor.Trigger(Variable=InitVariable.Trigger().Variable)
   n2 = UE.RandomTime(Next=^UE.Source.OnPlay, Reset=@n1.Value) {also {defaults {Max=35, Min=10}}}
   n3 = UE.TriggerRepeat(Start=@n2.On Next, Period=@n2.Value)
   n4 = &/MSS_Train_Exterior_Template(
    Enclosed.LPF=^Enclosed.LPF
    Enclosed.Volume=^Enclosed.Volume
    Movement.Speed=^Movement.Speed
    Pitch=^Pitch
    Variations=^"Track Stress "
    UE.Source.OnPlay=@n3.RepeatOut
   ) {also {defaults {Gain.PlayerOnTrain=0.7}}}
   n5 = &MSS_Train_Exterior_Template(
    Enclosed.LPF=^Enclosed.LPF
    Enclosed.Volume=^Enclosed.Volume
    Loop=^Loop
    Movement.Speed=^Movement.Speed
    Pitch=^Pitch
    UE.Source.OnPlay=^UE.Source.OnPlay
    Variations=^"Wheels "
   ) {also {defaults {Gain.PlayerOnTrain=0.7}}}
   AudioMixerAudioMixerStereo2_0 = "AudioMixer.Audio Mixer (Stereo, 2)"(
    In 1 L=@n4."UE.OutputFormat.Stereo.Audio:0"
    In 1 R=@n4."UE.OutputFormat.Stereo.Audio:1"
    In 0 L=@n5."UE.OutputFormat.Stereo.Audio:0"
    In 0 R=@n5."UE.OutputFormat.Stereo.Audio:1"
   ) {also {defaults {Gain 1=0.2}}}
   n6 = Modulation.GetModulatorValue() {also {defaults {Modulator=/Train/Sound/Modulation/CB_Train_Stopped}}}
   Multiply_0 = UE.Multiply:Audio by Float(PrimaryOperand=@AudioMixerAudioMixerStereo2_0.Out R, AdditionalOperands=@n6.Out)
   Multiply_1 = UE.Multiply:Audio by Float(PrimaryOperand=@AudioMixerAudioMixerStereo2_0.Out L, AdditionalOperands=@n6.Out)
   VariableMutator.Trigger(Value=@n3.RepeatOut, Variable=@n1.Variable)
  ]
  links [{dst="@UE.OutputFormat.Stereo.Audio:1", src=Multiply_0.Out}, {dst="@UE.OutputFormat.Stereo.Audio:0", src=Multiply_1.Out}]
 }
 Mobile {
  input_defaults {"Track Stress " []}
  nodes [
   InitVariable.Trigger()
   n1 = &MSS_Train_Exterior_Template(
    Enclosed.LPF=^Enclosed.LPF
    Enclosed.Volume=^Enclosed.Volume
    Loop=^Loop
    Movement.Speed=^Movement.Speed
    Pitch=^Pitch
    UE.Source.OnPlay=^UE.Source.OnPlay
    Variations=^"Wheels "
   ) {also {defaults {Gain.PlayerOnTrain=0.7}}}
   n2 = Modulation.GetModulatorValue() {also {defaults {Modulator=&CB_Train_Stopped}}}
   Multiply_0 = UE.Multiply:Audio by Float(PrimaryOperand=@n1."UE.OutputFormat.Stereo.Audio:1", AdditionalOperands=@n2.Out)
   Multiply_1 = UE.Multiply:Audio by Float(PrimaryOperand=@n1."UE.OutputFormat.Stereo.Audio:0", AdditionalOperands=@n2.Out)
  ]
  links [{dst="@UE.OutputFormat.Stereo.Audio:1", src=Multiply_0.Out}, {dst="@UE.OutputFormat.Stereo.Audio:0", src=Multiply_1.Out}]
 }
]
properties {
 assetImportData=&self:AssetImportData
 bLooping=true
 duration=10000
 numChannels=2
 outputFormat=Stereo
 sampleRate=48000
}
