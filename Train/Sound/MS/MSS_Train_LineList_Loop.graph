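# Looping quad MetaSound for moving-train ambience: passing-object loops and whooshes, plus rooftop wind when the player is on the roof.
# Two directional stereo loops, randomly timed and pitched whooshes, and modulator-triggered wind feed per-channel mixers into an azimuth panner following smoothed source orientation.
# Output gain combines smoothed movement speed with a train-stopped modulator.
# A Mobile page uses one whoosh player with fewer samples and simpler speed handling.
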MetaSoundSource /Train/Sound/MS/MSS_Train_LineList_Loop
inputs [
 Movement.Speed {type=Float}
 Movement.SpeedAzimuth {type=Float}
 Movement.SpeedElevation {type=Float}
 UE.Source.OnPlay {type=Trigger}
 UE.Source.Orientation.Azimuth {type=Float}
 UE.Source.Orientation.Elevation {type=Float}
 UE.Spatialization.Azimuth {type=Float}
 UE.Spatialization.Elevation {type=Float}
 Whooshes {type=WaveAsset:Array}
]
interfaces [Fort.Movement:1.0, UE.OutputFormat.Quad:1.0, UE.Source.Orientation:1.0, UE.Source:1.1, UE.Spatialization:1.0]
outputs [
 "UE.OutputFormat.Quad.Audio:0" {type=Audio}
 "UE.OutputFormat.Quad.Audio:1" {type=Audio}
 "UE.OutputFormat.Quad.Audio:4" {type=Audio}
 "UE.OutputFormat.Quad.Audio:5" {type=Audio}
]
pages [
 Default {
  is_build_graph=true
  input_defaults {
   Movement.Speed=0
   Movement.SpeedAzimuth=0
   Movement.SpeedElevation=0
   UE.Source.OnPlay=null
   UE.Source.Orientation.Azimuth=0
   UE.Source.Orientation.Elevation=0
   UE.Spatialization.Azimuth=0
   UE.Spatialization.Elevation=0
   Whooshes [
    /Train/Sound/Waves/Locomotion/S28/Train_ObjectPass_LeftToRight_01
    /Train/Sound/Waves/Locomotion/S28/Train_ObjectPass_LeftToRight_02
    /Train/Sound/Waves/Locomotion/S28/Train_ObjectPass_LeftToRight_03
    /Train/Sound/Waves/Locomotion/S28/Train_ObjectPass_LeftToRight_04
    /Train/Sound/Waves/Locomotion/S28/Train_ObjectPass_LeftToRight_05
    /Train/Sound/Waves/Locomotion/S28/Train_ObjectPass_LeftToRight_06
   ]
  }
  nodes [
   n1 = VariableDeferredAccessor.Trigger(Variable=InitVariable.Trigger().Variable)
   n2 = UE.RandomTime(Next=^UE.Source.OnPlay, Reset=@n1.Value) {also {defaults {Max=4, Min=1}}}
   n3 = UE.TriggerRepeat(Start=@n2.On Next, Period=@n2.Value)
   n4 = &MSS_PlayRandomSound_Stereo(Sounds=^Whooshes, UE.Source.OnPlay=@n3.RepeatOut) {also {defaults {MaxPitch=1.2, MinPitch=0.8, MinVolume=0.75}}}
   n5 = Modulation.GetModulatorValue() {also {defaults {Modulator=/Train/Sound/Settings/CB_PlayerIsOnTrainRoof}}}
   n6 = &/MSS_Train_Rooftop_Wind(
    UE.Source.OnPlay=UE.TriggerOnThreshold:Float(In=@n5.Out) {also {defaults {Threshold=0.5}}}.Out
    Stop=UE.TriggerOnThreshold:Float(In=@n5.Out) {also {defaults {Threshold=0.5, Type=1}}}.Out
   )
   n7 = VariableAccessor.Float(Variable=InitVariable.Float().Variable)
   n8 = VariableAccessor.Float(Variable=InitVariable.Float().Variable)
   WavePlayer_1 = UE.Wave Player:Stereo(Play=^UE.Source.OnPlay) {also {
     defaults {Loop="<MetaSoundBuilderResult.SUCCEEDED: 0>", Wave Asset=/Train/Sound/Waves/Locomotion/S28/Train_ObjectPass_LeftToRIght_Loop_01}
     version [1,1]
    }}
   AudioMixerAudioMixerMono3_0 = "AudioMixer.Audio Mixer (Mono, 3)"(
    In 1=@n4."UE.OutputFormat.Stereo.Audio:0"
    In 2=@n6."UE.OutputFormat.Quad.Audio:0"
    Gain 2=@n7.Value
    Gain 1=@n8.Value
    In 0=@WavePlayer_1.Out Left
   )
   n9 = VariableDeferredAccessor.Trigger(Variable=InitVariable.Trigger().Variable)
   n10 = UE.RandomTime(Next=^UE.Source.OnPlay, Reset=@n9.Value) {also {defaults {Max=4, Min=1}}}
   n11 = UE.TriggerRepeat(Start=@n10.On Next, Period=@n10.Value)
   n12 = &MSS_PlayRandomSound_Stereo(Sounds=^Whooshes, UE.Source.OnPlay=@n11.RepeatOut) {also {defaults {MaxPitch=1.2, MinPitch=0.8, MinVolume=0.75}}}
   WavePlayer_0 = UE.Wave Player:Stereo(Play=^UE.Source.OnPlay) {also {
     defaults {Loop="<MetaSoundBuilderResult.SUCCEEDED: 0>", Wave Asset=/Train/Sound/Waves/Locomotion/S28/Train_ObjectPass_RightToLeft_Loop_01}
     version [1,1]
    }}
   AudioMixerAudioMixerMono3_1 = "AudioMixer.Audio Mixer (Mono, 3)"(
    In 1=@n12."UE.OutputFormat.Stereo.Audio:1"
    In 2=@n6."UE.OutputFormat.Quad.Audio:5"
    Gain 2=@n7.Value
    Gain 1=@n8.Value
    In 0=@WavePlayer_0.Out Left
   )
   AudioMixerAudioMixerMono3_2 = "AudioMixer.Audio Mixer (Mono, 3)"(
    In 1=@n4."UE.OutputFormat.Stereo.Audio:1"
    In 2=@n6."UE.OutputFormat.Quad.Audio:4"
    Gain 2=@n7.Value
    Gain 1=@n8.Value
    In 0=@WavePlayer_1.Out Right
   )
   AudioMixerAudioMixerMono3_3 = "AudioMixer.Audio Mixer (Mono, 3)"(
    In 1=@n12."UE.OutputFormat.Stereo.Audio:0"
    In 2=@n6."UE.OutputFormat.Quad.Audio:1"
    Gain 2=@n7.Value
    Gain 1=@n8.Value
    In 0=@WavePlayer_0.Out Right
   )
   n13 = &MSP_AzimuthPannerQuad(
    In Front Left=@AudioMixerAudioMixerMono3_0.Out
    In Side Right=@AudioMixerAudioMixerMono3_1.Out
    In Side Left=@AudioMixerAudioMixerMono3_2.Out
    In Front Right=@AudioMixerAudioMixerMono3_3.Out
    Azimuth=UE.InterpTo:Audio(Target=^UE.Source.Orientation.Azimuth) {also {defaults {Interp Time=0.5}}}.Value
   )
   TriggerOnValueChange_0 = UE.Trigger On Value Change:Float(Value=^Movement.Speed)
   n14 = TriggerCompare.Clamp:Float(A=^Movement.Speed, Compare=@TriggerOnValueChange_0.Trigger) {also {defaults {Type=1}}}
   n15 = VariableAccessor.Float(Variable=VariableMutator.Float(Variable=InitVariable.Float().Variable, Value=Value.Value:Float(Target Value=^Movement.Speed, Set=@n14.True).Output Value).Variable)
   n16 = UE.Multiply:Float(
    AdditionalOperands=MapRange.MapRange:Float(In=UE.InterpTo:Audio(Target=@n15.Value) {also {defaults {Interp Time=0.5}}}.Value) {also {defaults {In Range B=1000}}}.Out Value
    PrimaryOperand=Modulation.GetModulatorValue() {also {defaults {Modulator=/Train/Sound/Modulation/CB_Train_Stopped}}}.Out
   )
   Multiply_0 = UE.Multiply:Audio by Float(PrimaryOperand=@n13.Out Front Left, AdditionalOperands=@n16.Out)
   Multiply_1 = UE.Multiply:Audio by Float(PrimaryOperand=@n13.Out Front Right, AdditionalOperands=@n16.Out)
   Multiply_2 = UE.Multiply:Audio by Float(PrimaryOperand=@n13.Out Side Right, AdditionalOperands=@n16.Out)
   Multiply_3 = UE.Multiply:Audio by Float(PrimaryOperand=@n13.Out Side Left, AdditionalOperands=@n16.Out)
   VariableMutator.Trigger(Value=@n3.RepeatOut, Variable=@n1.Variable)
   VariableMutator.Trigger(Value=@n11.RepeatOut, Variable=@n9.Variable)
  ]
  links [
   {dst="@UE.OutputFormat.Quad.Audio:0", src=Multiply_0.Out}
   {dst="@UE.OutputFormat.Quad.Audio:1", src=Multiply_1.Out}
   {dst="@UE.OutputFormat.Quad.Audio:5", src=Multiply_2.Out}
   {dst="@UE.OutputFormat.Quad.Audio:4", src=Multiply_3.Out}
  ]
 }
 Mobile {
  input_defaults {Whooshes [
    &Train_ObjectPass_LeftToRight_01
    &Train_ObjectPass_LeftToRight_02
    &Train_ObjectPass_LeftToRight_03
   ]}
  nodes [
   InitVariable.Trigger()
   n1 = Modulation.GetModulatorValue() {also {defaults {Modulator=&CB_PlayerIsOnTrainRoof}}}
   n2 = &MSS_Train_Rooftop_Wind(
    UE.Source.OnPlay=UE.TriggerOnThreshold:Float(In=@n1.Out) {also {defaults {Threshold=0.5}}}.Out
    Stop=UE.TriggerOnThreshold:Float(In=@n1.Out) {also {defaults {Threshold=0.5, Type=1}}}.Out
   )
   n3 = VariableAccessor.Float(Variable=InitVariable.Float().Variable)
   WavePlayer_0 = UE.Wave Player:Stereo(Play=^UE.Source.OnPlay) {also {
     defaults {Loop="<MetaSoundBuilderResult.SUCCEEDED: 0>", Wave Asset=&Train_ObjectPass_RightToLeft_Loop_01}
     version [1,1]
    }}
   AudioMixerAudioMixerMono2_0 = "AudioMixer.Audio Mixer (Mono, 2)"(In 1=@n2."UE.OutputFormat.Quad.Audio:5", Gain 1=@n3.Value, In 0=@WavePlayer_0.Out Left)
   AudioMixerAudioMixerMono2_1 = "AudioMixer.Audio Mixer (Mono, 2)"(In 1=@n2."UE.OutputFormat.Quad.Audio:1", Gain 1=@n3.Value, In 0=@WavePlayer_0.Out Right)
   n4 = VariableDeferredAccessor.Trigger(Variable=InitVariable.Trigger().Variable)
   n5 = UE.RandomTime(Next=^UE.Source.OnPlay, Reset=@n4.Value) {also {defaults {Max=4, Min=1}}}
   n6 = UE.TriggerRepeat(Start=@n5.On Next, Period=@n5.Value)
   n7 = &MSS_PlayRandomSound_Stereo(Sounds=^Whooshes, UE.Source.OnPlay=@n6.RepeatOut) {also {defaults {MaxPitch=1.2, MinPitch=0.8, MinVolume=0.75}}}
   n8 = VariableAccessor.Float(Variable=InitVariable.Float().Variable)
   WavePlayer_1 = UE.Wave Player:Stereo(Play=^UE.Source.OnPlay) {also {
     defaults {Loop="<MetaSoundBuilderResult.SUCCEEDED: 0>", Wave Asset=&Train_ObjectPass_LeftToRIght_Loop_01}
     version [1,1]
    }}
   AudioMixerAudioMixerMono3_0 = "AudioMixer.Audio Mixer (Mono, 3)"(
    In 1=@n7."UE.OutputFormat.Stereo.Audio:0"
    In 2=@n2."UE.OutputFormat.Quad.Audio:0"
    Gain 2=@n3.Value
    Gain 1=@n8.Value
    In 0=@WavePlayer_1.Out Left
   )
   AudioMixerAudioMixerMono3_1 = "AudioMixer.Audio Mixer (Mono, 3)"(
    In 1=@n7."UE.OutputFormat.Stereo.Audio:1"
    In 2=@n2."UE.OutputFormat.Quad.Audio:4"
    Gain 2=@n3.Value
    Gain 1=@n8.Value
    In 0=@WavePlayer_1.Out Right
   )
   n9 = &MSP_AzimuthPannerQuad(
    In Side Right=@AudioMixerAudioMixerMono2_0.Out
    In Front Right=@AudioMixerAudioMixerMono2_1.Out
    In Front Left=@AudioMixerAudioMixerMono3_0.Out
    In Side Left=@AudioMixerAudioMixerMono3_1.Out
    Azimuth=UE.InterpTo:Audio(Target=^UE.Source.Orientation.Azimuth) {also {defaults {Interp Time=0.5}}}.Value
   )
   n10 = UE.Multiply:Float(
    AdditionalOperands=MapRange.MapRange:Float(In=UE.InterpTo:Audio(Target=^Movement.Speed) {also {defaults {Interp Time=0.5}}}.Value) {also {defaults {In Range B=1000}}}.Out Value
    PrimaryOperand=Modulation.GetModulatorValue() {also {defaults {Modulator=&CB_Train_Stopped}}}.Out
   )
   Multiply_0 = UE.Multiply:Audio by Float(PrimaryOperand=@n9.Out Front Left, AdditionalOperands=@n10.Out)
   Multiply_1 = UE.Multiply:Audio by Float(PrimaryOperand=@n9.Out Front Right, AdditionalOperands=@n10.Out)
   Multiply_2 = UE.Multiply:Audio by Float(PrimaryOperand=@n9.Out Side Right, AdditionalOperands=@n10.Out)
   Multiply_3 = UE.Multiply:Audio by Float(PrimaryOperand=@n9.Out Side Left, AdditionalOperands=@n10.Out)
   VariableMutator.Trigger(Value=@n6.RepeatOut, Variable=@n4.Variable)
  ]
  links [
   {dst="@UE.OutputFormat.Quad.Audio:0", src=Multiply_0.Out}
   {dst="@UE.OutputFormat.Quad.Audio:1", src=Multiply_1.Out}
   {dst="@UE.OutputFormat.Quad.Audio:5", src=Multiply_2.Out}
   {dst="@UE.OutputFormat.Quad.Audio:4", src=Multiply_3.Out}
  ]
 }
]
properties {
 assetImportData=&self:AssetImportData
 bLooping=true
 duration=10000
 numChannels=4
 outputFormat=Quad
 sampleRate=48000
 modulationSettings {volumeRouting=Union}
}
