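# Looping Sound Cue for train exterior audio, routed to the Fort_Vehicles_Exterior sound class.
# A quality level node picks the LineList loop for the first two levels and the exterior wheels loop for the last two, both MetaSound sources by name.
# Each loop has its own attenuation asset.
# A concurrency override caps voices at four, with distance-based volume scaling and timed voice-steal release.
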SoundCue /Train/Sound/Cues/Cue_Train_LineList_Loop
root=QualityLevel_1
nodes [
 {id=Attenuation_0, children [WavePlayer_3], properties {attenuationSettings=/Train/Sound/Settings/Train_LineList_Att}}
 {id=Attenuation_1, children [WavePlayer_0], properties {attenuationSettings=/Train/Sound/Settings/Train_Movement_Wheels_Att}}
 {id=QualityLevel_1, children [Attenuation_0, Attenuation_0, Attenuation_1, Attenuation_1]}
 {id=WavePlayer_0, properties {bLooping=true, soundWaveAssetPtr=/Train/Sound/MS/MSS_Train_Exterior_Loop_Wheels}}
 {id=WavePlayer_3, properties {bLooping=true, soundWaveAssetPtr=/Train/Sound/MS/MSS_Train_LineList_Loop}}
]
properties {
 bDebug=true
 bOverrideConcurrency=true
 maxDistance=4000
 soundClassObject=&Fort_Vehicles_Exterior
 volumeMultiplier=1
 attenuationOverrides {falloffDistance=7124.0664, stereoSpread=0.606}
 concurrencyOverrides {bVolumeScaleCanRelease=true, maxCount=4, voiceStealReleaseTime=0.5, volumeScale=0.75, volumeScaleAttackTime=0.5, volumeScaleMode=Distance}
}
