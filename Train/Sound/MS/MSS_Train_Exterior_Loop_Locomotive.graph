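# Stereo MetaSound source for the train locomotive's exterior loop, layering engine idle, power up/down and air-release sounds.
# Three MSS_Train_Exterior_Template instances share speed, pitch and enclosure inputs and feed a stereo mixer.
# The engine layer plays on start; the other two retrigger at periods randomized once on play.
# The Mobile page keeps only the engine layer, bypassing the mixer and emptying the other wave arrays.
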MetaSoundSource /Train/Sound/MS/MSS_Train_Exterior_Loop_Locomotive
inputs [
 Air_Steam {type=WaveAsset:Array}
 Enclosed.LPF {type=Float}
 Enclosed.Volume {type=Float}
 Engine {type=WaveAsset:Array}
 Loop {type=Bool}
 Movement.Speed {type=Float}
 Movement.SpeedAzimuth {type=Float}
 Movement.SpeedElevation {type=Float}
 Pitch {type=Float}
 PowerUpandDown {type=WaveAsset:Array}
 UE.Source.OnPlay {type=Trigger}
]
interfaces [Fort.Movement:1.0, UE.OutputFormat.Stereo:1.0, UE.Source:1.1]
outputs [
 "UE.OutputFormat.Stereo.Audio:0" {type=Audio}
 "UE.OutputFormat.Stereo.Audio:1" {type=Audio}
]
pages [
 Default {
  is_build_graph=true
  input_defaults {
   Enclosed.LPF=6000
   Enclosed.Volume=0.3
   Loop="<MetaSoundBuilderResult.SUCCEEDED: 0>"
   Movement.Speed=0
   Movement.SpeedAzimuth=0
   Movement.SpeedElevation=0
   Pitch=1
   UE.Source.OnPlay=null
   Air_Steam [
    /Train/Sound/Waves/Locomotion/S28/Train_EngineAirRelease_03
    /Train/Sound/Waves/Locomotion/S28/Train_EngineAirRelease_02
    /Train/Sound/Waves/Locomotion/S28/Train_EngineAirRelease_01
    /Train/Sound/Waves/Locomotion/S28/Train_EngineAirRelease_04
    /Train/Sound/Waves/Locomotion/S28/Train_EngineAirRelease_05
   ]
   Engine [/Train/Sound/Waves/Locomotion/S28/Train_Movement_Idle]
   PowerUpandDown [/Train/Sound/Waves/Locomotion/S28/Train_Engine_PowerUP, /Train/Sound/Waves/Locomotion/S28/Train_Engine_PowerDown]
  }
  nodes [
   MSS_Train_Exterior_Template_0 = &/MSS_Train_Exterior_Template(
    Enclosed.LPF=^Enclosed.LPF
    Enclosed.Volume=^Enclosed.Volume
    Variations=^Engine
    Loop=^Loop
    Movement.Speed=^Movement.Speed
    Pitch=^Pitch
    UE.Source.OnPlay=^UE.Source.OnPlay
   ) {also {defaults {Gain.PlayerOffTrain=1.2, Gain.PlayerOnTrain=0.7, VolumeMinSpeed=0.2}}}
   n1 = &MSS_Train_Exterior_Template(
    Enclosed.LPF=^Enclosed.LPF
    Enclosed.Volume=^Enclosed.Volume
    Movement.Speed=^Movement.Speed
    Pitch=^Pitch
    Variations=^PowerUpandDown
    UE.Source.OnPlay=UE.TriggerRepeat(Start=^UE.Source.OnPlay, Period=UE.RandomTime(Next=^UE.Source.OnPlay) {also {defaults {Max=50, Min=30}}}.Value).RepeatOut
   ) {also {defaults {Gain.PlayerOnTrain=0.7}}}
   n2 = &MSS_Train_Exterior_Template(
    Variations=^Air_Steam
    Enclosed.LPF=^Enclosed.LPF
    Enclosed.Volume=^Enclosed.Volume
    Movement.Speed=^Movement.Speed
    Pitch=^Pitch
    UE.Source.OnPlay=UE.TriggerRepeat(Start=^UE.Source.OnPlay, Period=UE.RandomTime(Next=^UE.Source.OnPlay) {also {defaults {Max=40, Min=15}}}.Value).RepeatOut
   ) {also {defaults {Gain.PlayerOnTrain=0.7, VolumeMinSpeed=0.5}}}
   AudioMixerAudioMixerStereo3_0 = "AudioMixer.Audio Mixer (Stereo, 3)"(
    In 0 L=@MSS_Train_Exterior_Template_0."UE.OutputFormat.Stereo.Audio:0"
    In 0 R=@MSS_Train_Exterior_Template_0."UE.OutputFormat.Stereo.Audio:1"
    In 2 L=@n1."UE.OutputFormat.Stereo.Audio:0"
    In 2 R=@n1."UE.OutputFormat.Stereo.Audio:1"
    In 1 L=@n2."UE.OutputFormat.Stereo.Audio:0"
    In 1 R=@n2."UE.OutputFormat.Stereo.Audio:1"
   ) {also {defaults {Gain 1=0.55, Gain 2=0.7}}}
  ]
  links [
   {dst="@UE.OutputFormat.Stereo.Audio:0", src=AudioMixerAudioMixerStereo3_0.Out L}
   {dst="@UE.OutputFormat.Stereo.Audio:1", src=AudioMixerAudioMixerStereo3_0.Out R}
  ]
 }
 Mobile {
  input_defaults {Air_Steam [], PowerUpandDown []}
  nodes [MSS_Train_Exterior_Template_0 = &MSS_Train_Exterior_Template(
    Enclosed.LPF=^Enclosed.LPF
    Enclosed.Volume=^Enclosed.Volume
    Variations=^Engine
    Loop=^Loop
    Movement.Speed=^Movement.Speed
    Pitch=^Pitch
    UE.Source.OnPlay=^UE.Source.OnPlay
   ) {also {defaults {Gain.PlayerOffTrain=1.2, Gain.PlayerOnTrain=0.7, VolumeMinSpeed=0.2}}}]
  links [
   {dst="@UE.OutputFormat.Stereo.Audio:0", src=MSS_Train_Exterior_Template_0.UE.OutputFormat.Stereo.Audio:0}
   {dst="@UE.OutputFormat.Stereo.Audio:1", src=MSS_Train_Exterior_Template_0.UE.OutputFormat.Stereo.Audio:1}
  ]
 }
]
properties {
 assetImportData=&self:AssetImportData
 bLooping=true
 duration=10000
 numChannels=2
 outputFormat=Stereo
 sampleRate=48000
}
