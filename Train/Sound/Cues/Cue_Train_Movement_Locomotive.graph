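# Sound cue for the locomotive's exterior movement audio, mixing four separately attenuated layers into one output.
# Looping MetaSounds supply near engine and wheel layers, with the wheels pitched up an octave and shaped by a volume envelope that fades in.
# A distant loop plays alongside randomly chosen distant one-shots, repeated after random delays with varied pitch and volume.
# It uses a train source effect chain and references a first-person gunfire concurrency set.
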SoundCue /Train/Sound/Cues/Cue_Train_Movement_Locomotive
root=Modulator_0
nodes [
 {id=Attenuation_0, children [Modulator_2], properties {attenuationSettings=/Train/Sound/Settings/Train_Movement_Distant_Att}}
 {id=Attenuation_1, children [Modulator_1], properties {attenuationSettings=/Train/Sound/Settings/Train_Movement_Att}}
 {id=Attenuation_2, children [Modulator_4], properties {attenuationSettings=&Train_Movement_Att}}
 {id=Attenuation_3, children [Looping_0], properties {attenuationSettings=/Train/Sound/Settings/Train_Movement_Distant_OneShots_Att}}
 {id=Delay_0, children [Modulator_3], properties {delayMax=10, delayMin=2}}
 {
  id=Enveloper_1
  children [WavePlayer_17]
  properties {pitchCurve {editorCurveData {keys [{time=0}]}}, volumeCurve {editorCurveData {keys [<at 0 0>,<at 7 1>]}}}
 }
 {id=Looping_0, children [Delay_0]}
 {id=Mixer_0, children [Attenuation_1, Attenuation_0, Attenuation_3, Attenuation_2, null], properties {inputVolume [4,1.5,0.4,4,3.5]}}
 {id=Modulator_0, children [Mixer_0], properties {pitchMax=1, pitchMin=1}}
 {id=Modulator_1, children [WavePlayer_16], properties {pitchMax=1, pitchMin=1, volumeMax=1, volumeMin=1}}
 {id=Modulator_2, children [WavePlayer_0], properties {pitchMax=1, pitchMin=1, volumeMax=1, volumeMin=1}}
 {id=Modulator_3, children [Random_0], properties {pitchMax=1.5, volumeMax=1, volumeMin=0.2}}
 {id=Modulator_4, children [Enveloper_1], properties {pitchMax=2, pitchMin=2, volumeMax=1, volumeMin=1}}
 {
  id=Random_0
  children [WavePlayer_9, WavePlayer_10, WavePlayer_11, WavePlayer_12, WavePlayer_13, WavePlayer_14, WavePlayer_15, WavePlayer_18]
  properties {bShouldExcludeFromBranchCulling=true, weights [1,1,1,1,1,1,1,1]}
 }
 {id=WavePlayer_0, properties {bLooping=true, soundWaveAssetPtr=/Train/Sound/Waves/Locomotion/S28/Train_Movement_Distant_Loop}}
 {id=WavePlayer_10, properties {soundWaveAssetPtr=/Train/Sound/Waves/Locomotion/S28/Train_Distant_Movement_02}}
 {id=WavePlayer_11, properties {soundWaveAssetPtr=/Train/Sound/Waves/Locomotion/S28/Train_Distant_Movement_03}}
 {id=WavePlayer_12, properties {soundWaveAssetPtr=/Train/Sound/Waves/Locomotion/S28/Train_Distant_Movement_04}}
 {id=WavePlayer_13, properties {soundWaveAssetPtr=/Train/Sound/Waves/Locomotion/S28/Train_Distant_Movement_05}}
 {id=WavePlayer_14, properties {soundWaveAssetPtr=/Train/Sound/Waves/Locomotion/S28/Train_Distant_Movement_06}}
 {id=WavePlayer_15, properties {soundWaveAssetPtr=/Train/Sound/Waves/Locomotion/S28/Train_Distant_Movement_07}}
 {id=WavePlayer_16, properties {bLooping=true, soundWaveAssetPtr=/Train/Sound/MS/MSS_Train_Exterior_Loop_Locomotive}}
 {id=WavePlayer_17, properties {bLooping=true, soundWaveAssetPtr=/Train/Sound/MS/MSS_Train_Exterior_Loop_Wheels}}
 {id=WavePlayer_18, properties {soundWaveAssetPtr=/Train/Sound/Waves/Locomotion/S28/Train_Distant_Movement_08}}
 {id=WavePlayer_9, properties {soundWaveAssetPtr=/Train/Sound/Waves/Locomotion/S28/Train_Distant_Movement_01}}
]
properties {
 bDebug=true
 maxDistance=45000
 soundClassObject=&Fort_Vehicles_Exterior
 sourceEffectChain=/Train/Sound/Settings/SEPC_Train
 volumeMultiplier=1.2
 attenuationOverrides {bSpatialize=false}
 concurrencyOverrides {bLimitToOwner=true, maxCount=3, resolutionRule=StopOldest}
 concurrencySet [&Weapon_Gunfire_FirstPerson_Conc]
}
