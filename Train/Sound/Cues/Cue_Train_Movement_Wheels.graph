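# Sound cue for the looping exterior wheel sound of a moving train, routed to the exterior vehicles sound class.
# A looping MetaSound wave player feeds a modulator that randomizes pitch upward, which fills the first two quality-level slots; the other two are empty.
# An attenuation asset and train source effect chain shape the sound.
# Overridden concurrency caps simultaneous instances, with distance-based volume scaling and soft voice-steal release.
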SoundCue /Train/Sound/Cues/Cue_Train_Movement_Wheels
root=QualityLevel_0
nodes [
 {id=Modulator_4, children [WavePlayer_3], properties {pitchMax=2, pitchMin=1, volumeMax=1, volumeMin=1}}
 {id=QualityLevel_0, children [Modulator_4,Modulator_4,null,null]}
 {id=WavePlayer_3, properties {bLooping=true, soundWaveAssetPtr=/Train/Sound/MS/MSS_Train_Exterior_Loop_Wheels}}
]
properties {
 attenuationSettings=/Train/Sound/Settings/Train_Movement_Wheels_Att
 bDebug=true
 bOverrideConcurrency=true
 maxDistance=4000
 soundClassObject=&Fort_Vehicles_Exterior
 sourceEffectChain=/Train/Sound/Settings/SEPC_Train
 volumeMultiplier=2
 attenuationOverrides {bSpatialize=false}
 concurrencyOverrides {bVolumeScaleCanRelease=true, maxCount=4, voiceStealReleaseTime=0.5, volumeScale=0.75, volumeScaleAttackTime=0.5, volumeScaleMode=Distance}
}
