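# Sound cue for the player being launched in the Train content, layering four randomized sounds into one mix.
# A Mixer blends a briefly delayed launch-pad jump layer with thud, metal and whoosh layers, each weighted by input volume.
# Each layer picks a random wave variant, with Modulators varying the thud, metal and whoosh.
# It uses physics-impact attenuation, the player foley sound class, and per-owner single-instance concurrency that stops the oldest.
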SoundCue /Train/Sound/Cues/Cue_Train_Player_Launch
root=Mixer_1
nodes [
 {id=Delay_0, children [Random_0], properties {delayMax=0.15, delayMin=0.15}}
 {id=Mixer_1, children [Delay_0,Modulator_1,Modulator_0,Modulator_2], properties {inputVolume [2,4,3,2]}}
 {id=Modulator_0, children [Random_1]}
 {id=Modulator_1, children [Random_2]}
 {id=Modulator_2, children [Random_3]}
 {id=Random_0, children [WavePlayer_0,WavePlayer_1,WavePlayer_4], properties {weights [1,1,1]}}
 {id=Random_1, children [WavePlayer_2,WavePlayer_3,WavePlayer_5], properties {weights [1,1,1]}}
 {id=Random_2, children [WavePlayer_6,WavePlayer_7], properties {weights [1,1]}}
 {id=Random_3, children [WavePlayer_8, WavePlayer_9, WavePlayer_10, WavePlayer_11], properties {weights [1,1,1,1]}}
 {id=WavePlayer_0, properties {soundWaveAssetPtr=/Game/Athena/Sounds/LaunchPad/sfx_LaunchPad_JumpLayer_01}}
 {id=WavePlayer_1, properties {soundWaveAssetPtr=/Game/Athena/Sounds/LaunchPad/sfx_LaunchPad_JumpLayer_02}}
 {id=WavePlayer_10, properties {soundWaveAssetPtr=/Train/Sound/Waves/Locomotion/S28/Train_PlayerLaunch_Whoosh_03}}
 {id=WavePlayer_11, properties {soundWaveAssetPtr=/Train/Sound/Waves/Locomotion/S28/Train_PlayerLaunch_Whoosh_04}}
 {id=WavePlayer_2, properties {soundWaveAssetPtr=/Train/Sound/Waves/Locomotion/S28/Train_PlayerLaunch_Metal_01}}
 {id=WavePlayer_3, properties {soundWaveAssetPtr=/Train/Sound/Waves/Locomotion/S28/Train_PlayerLaunch_Metal_02}}
 {id=WavePlayer_4, properties {soundWaveAssetPtr=/Game/Athena/Sounds/LaunchPad/sfx_LaunchPad_JumpLayer_03}}
 {id=WavePlayer_5, properties {soundWaveAssetPtr=/Train/Sound/Waves/Locomotion/S28/Train_PlayerLaunch_Metal_03}}
 {id=WavePlayer_6, properties {soundWaveAssetPtr=/Train/Sound/Waves/Locomotion/S28/Train_PlayerLaunch_Thud_01}}
 {id=WavePlayer_7, properties {soundWaveAssetPtr=/Train/Sound/Waves/Locomotion/S28/Train_PlayerLaunch_Thud_02}}
 {id=WavePlayer_8, properties {soundWaveAssetPtr=/Train/Sound/Waves/Locomotion/S28/Train_PlayerLaunch_Whoosh_01}}
 {id=WavePlayer_9, properties {soundWaveAssetPtr=/Train/Sound/Waves/Locomotion/S28/Train_PlayerLaunch_Whoosh_02}}
]
properties {
 attenuationSettings=/Game/Sounds/Physics/Settings/Physics_Impact_Att
 bDebug=true
 bOverrideConcurrency=true
 maxDistance=8512
 soundClassObject=&Fort_Player_Foley
 volumeMultiplier=1
 attenuationOverrides {bSpatialize=false}
 concurrencyOverrides {bLimitToOwner=true, maxCount=1, resolutionRule=StopOldest, retriggerTime=0.75}
}
