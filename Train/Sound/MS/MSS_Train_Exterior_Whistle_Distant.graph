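# Stereo MetaSound preset for the distant exterior train whistle, playing a Japanese distant horn wave.
# It inherits its graph from MSS_Train_Exterior_Template, which exposes inputs for enclosed filtering and volume, on/off-train gain, pitch, looping and speed-based volume.
# It implements the Fort.Movement and one-shot interfaces.
# Source volume is raised above unity and modulated by the CB_Train_Horn modulator.
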MetaSoundSource /Train/Sound/MS/MSS_Train_Exterior_Whistle_Distant
preset_parent=&/MSS_Train_Exterior_Template
inputs [
 Enclosed.LPF {type=Float}
 Enclosed.Volume {type=Float}
 Gain.PlayerOffTrain {type=Float}
 Gain.PlayerOnTrain {type=Float}
 Loop {type=Bool}
 MaxTrainSpeed {type=Float}
 Movement.Speed {type=Float}
 Movement.SpeedAzimuth {type=Float}
 Movement.SpeedElevation {type=Float}
 Pitch {type=Float}
 UE.Source.OnPlay {type=Trigger}
 Variations {type=WaveAsset:Array}
 VelocityMod {type=Bool}
 VolumeMaxSpeed {type=Float}
 VolumeMinSpeed {type=Float}
]
interfaces [Fort.Movement:1.0, UE.OutputFormat.Stereo:1.0, UE.Source.OneShot:1.0, UE.Source:1.1]
outputs [
 "UE.OutputFormat.Stereo.Audio:0" {type=Audio}
 "UE.OutputFormat.Stereo.Audio:1" {type=Audio}
 UE.Source.OneShot.OnFinished {type=Trigger}
]
pages [Default {input_defaults {
   Enclosed.LPF=6000
   Enclosed.Volume=0.1
   Gain.PlayerOffTrain=1
   Gain.PlayerOnTrain=1
   Loop=null
   MaxTrainSpeed=1000
   Movement.Speed=0
   Movement.SpeedAzimuth=0
   Movement.SpeedElevation=0
   Pitch=1
   UE.Source.OnPlay=null
   VelocityMod=null
   VolumeMaxSpeed=1
   VolumeMinSpeed=0
   Variations [/Train/Sound/Waves/Japan_Train/Train_Horn_Distant_Japan]
  }}]
properties {
 assetImportData=&self:AssetImportData
 numChannels=2
 outputFormat=Stereo
 sampleRate=48000
 volume=1.5
 modulationSettings {volumeRouting=Union, volumeModulationDestination {bEnableModulation=true, modulators [/Train/Sound/Modulation/CB_Train_Horn]}}
}
