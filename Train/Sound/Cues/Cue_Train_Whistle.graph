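# Sound cue for the train's exterior whistle, layering close and distant versions.
# A mixer blends two branches, each with its own attenuation asset, with the close layer boosted and passed through a modulator with pitch variation flattened.
# Both wave players reference MSS-prefixed assets, likely MetaSound sources.
# The cue uses the exterior vehicle sound class, a train source effect chain and a borrowed first-person gunfire concurrency set, with debug enabled.
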SoundCue /Train/Sound/Cues/Cue_Train_Whistle
root=Mixer_1
nodes [
 {id=Attenuation_3, children [Modulator_2], properties {attenuationSettings=/Train/Sound/Settings/Train_Whistle_Close_Att}}
 {id=Attenuation_4, children [WavePlayer_1], properties {attenuationSettings=/Train/Sound/Settings/Train_Whistle_Distant_Att}}
 {id=Mixer_1, children [Attenuation_3, Attenuation_4], properties {inputVolume [1.5,1]}}
 {id=Modulator_2, children [WavePlayer_18], properties {pitchMax=1, pitchMin=1}}
 {id=WavePlayer_1, properties {soundWaveAssetPtr=/Train/Sound/MS/MSS_Train_Exterior_Whistle_Distant}}
 {id=WavePlayer_18, properties {soundWaveAssetPtr=/Train/Sound/MS/MSS_Train_Exterior_Whistle}}
]
properties {
 bDebug=true
 maxDistance=51000
 soundClassObject=&Fort_Vehicles_Exterior
 sourceEffectChain=/Train/Sound/Settings/SEPC_Train
 volumeMultiplier=2
 attenuationOverrides {bSpatialize=false}
 concurrencyOverrides {bLimitToOwner=true, maxCount=3, resolutionRule=StopOldest}
 concurrencySet [&Weapon_Gunfire_FirstPerson_Conc]
}
