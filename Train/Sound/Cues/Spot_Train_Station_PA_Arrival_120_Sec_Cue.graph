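# Sound cue for a train station PA arrival announcement, named as a 120-second spot, with no nodes or root, so it has no sound to play as it stands.
# Base submix and submix sends are off, while a pre-effect send feeds the station announcement audio bus at partial level.
# Concurrency overrides cap simultaneous instances and set a retrigger time, and the maximum distance is very large.
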SoundCue /Train/Sound/Cues/Spot_Train_Station_PA_Arrival_120_Sec_Cue
root=null
nodes []
properties {
 bEnableBaseSubmix=false
 bEnableSubmixSends=false
 maxDistance=2097152
 volumeMultiplier=1
 concurrencyOverrides {maxCount=2, retriggerTime=3}
 preEffectBusSends [{audioBus=/Train/Sound/AB_TrainStationAnnouncmentBus, sendLevel=0.6, soundSourceBus=null}]
}
